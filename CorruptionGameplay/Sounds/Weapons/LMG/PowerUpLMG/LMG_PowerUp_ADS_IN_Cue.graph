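# Sound cue for the power-up LMG's aim-down-sights-in foley, classed as weapon foley and attenuated with the shared weapon reload settings.
# A mixer blends a quiet fixed base wave with a louder layer picked at random from two equally weighted variants.
# A root modulator with default settings sits over the mix, and the whole cue plays at reduced volume.
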SoundCue /CorruptionGameplay/Sounds/Weapons/LMG/PowerUpLMG/LMG_PowerUp_ADS_IN_Cue
root=Modulator_0
nodes [
 {id=Mixer_0, children [WavePlayer_1,Random_0], properties {inputVolume [0.1,1.4]}}
 {id=Modulator_0, children [Mixer_0]}
 {id=Random_0, children [WavePlayer_4,WavePlayer_5], properties {weights [1,1]}}
 {id=WavePlayer_1, properties {soundWaveAssetPtr=&/LMG_PowerUp_ADS_In_02}}
 {id=WavePlayer_4, properties {soundWaveAssetPtr=&/LMG_PowerUp_ADS_In_Layer_01}}
 {id=WavePlayer_5, properties {soundWaveAssetPtr=&/LMG_PowerUp_ADS_In_Layer_02}}
]
properties {
 attenuationSettings=&Fort_Weapon_Reload_Att
 maxDistance=2250
 soundClassObject=&Fort_WeaponFoley
 volumeMultiplier=0.301
}
