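# Sound cue for the power-up LMG's weapon foley when leaving aim-down-sights.
# A Mixer plays the main ADS-out wave at reduced level alongside a full-level layer wave.
# A root Modulator varies the result and raises its minimum pitch above normal.
# It uses the weapon foley sound class and borrows the weapon reload attenuation, with a low overall volume.
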SoundCue /CorruptionGameplay/Sounds/Weapons/LMG/PowerUpLMG/LMG_PowerUp_ADS_OUT_Cue
root=Modulator_0
nodes [
 {id=Mixer_0, children [WavePlayer_0,WavePlayer_3], properties {inputVolume [0.6,1]}}
 {id=Modulator_0, children [Mixer_0], properties {pitchMin=1.05}}
 {id=WavePlayer_0, properties {soundWaveAssetPtr=&/LMG_PowerUp_ADS_Out}}
 {id=WavePlayer_3, properties {soundWaveAssetPtr=&/LMG_PowerUp_ADS_Out_Layer}}
]
properties {
 attenuationSettings=&Fort_Weapon_Reload_Att
 maxDistance=2250
 soundClassObject=&Fort_WeaponFoley
 volumeMultiplier=0.2
}
