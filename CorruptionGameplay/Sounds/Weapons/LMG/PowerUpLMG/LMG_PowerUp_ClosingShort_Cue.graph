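# Sound cue for the short closing foley of the LMG power-up weapon.
# A single closing foley wave plays through a Modulator that pins pitch at normal, at a reduced cue volume.
# It uses the player foley sound class with weapon pickup attenuation.
# It references a shared pickup concurrency set and carries override values for a single instance that blocks new plays.
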SoundCue /CorruptionGameplay/Sounds/Weapons/LMG/PowerUpLMG/LMG_PowerUp_ClosingShort_Cue
root=Modulator_1
nodes [
 {id=Modulator_1, children [WavePlayer_0], properties {pitchMax=1, pitchMin=1}}
 {id=WavePlayer_0, properties {soundWaveAssetPtr=&/LMG_Powerup_FoleyClosing_02}}
]
properties {
 attenuationSettings=&Fort_Weapon_Pickup_Att
 maxDistance=3500
 soundClassObject=&Fort_Player_Foley
 volumeMultiplier=0.55
 concurrencyOverrides {maxCount=1, resolutionRule=PreventNew}
 concurrencySet [&PickupMaximum_Oldest]
}
