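# Sound cue for the slow looping fire of the power-up LMG, layered by distance into close, medium and distant looping waves.
# Each layer has its own attenuation, with the close one using a third-person fire sound class and the others a distant-fire class.
# Medium and distant waves are pitched down and boosted, then close and medium share a further pitch modulator before mixing with distant.
# Global and local gunfire concurrency sets limit playback.
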SoundCue /CorruptionGameplay/Sounds/Weapons/LMG/PowerUpLMG/LMG_PowerUp_Fire_LoopSlow_Cue
root=Mixer_2
nodes [
 {id=Attenuation_3, children [WavePlayer_2], properties {attenuationSettings=&Fort_Weapon_Projectile_Close_Athena_Att}}
 {
  id=Attenuation_4
  children [Modulator_0]
  properties {attenuationSettings=&Fort_Weapon_Projectile_Distant_Mono_Athena_Att}
 }
 {
  id=Attenuation_5
  children [Modulator_2]
  properties {attenuationSettings=&Fort_Weapon_Projectile_Medium_Mono_Athena_Att}
 }
 {id=Mixer_2, children [Modulator_3,SoundClass_4], properties {inputVolume [1,0.5]}}
 {id=Mixer_3, children [SoundClass_3,SoundClass_5], properties {inputVolume [0.95,0.5]}}
 {id=Modulator_0, children [WavePlayer_8], properties {pitchMax=0.8, pitchMin=0.8, volumeMax=2, volumeMin=2}}
 {id=Modulator_2, children [WavePlayer_7], properties {pitchMax=0.8, pitchMin=0.8, volumeMax=2, volumeMin=2}}
 {id=Modulator_3, children [Mixer_3], properties {pitchMin=1.05}}
 {
  id=SoundClass_3
  children [Attenuation_3]
  properties {soundClassOverride=&Fort_WeaponRanged_Fire_3P_PSM}
 }
 {
  id=SoundClass_4
  children [Attenuation_4]
  properties {soundClassOverride=&Fort_WeaponRanged_Fire_Distant}
 }
 {
  id=SoundClass_5
  children [Attenuation_5]
  properties {soundClassOverride=&Fort_WeaponRanged_Fire_Distant}
 }
 {
  id=WavePlayer_2
  properties {bLooping=true, soundWaveAssetPtr=&/LMG_Powerup_Fire_LoopSlow_01}
 }
 {
  id=WavePlayer_7
  properties {bLooping=true, soundWaveAssetPtr=&/LMG_PowerUp_Fire_LoopSlow_FireLoop_Medium_01}
 }
 {
  id=WavePlayer_8
  properties {bLooping=true, soundWaveAssetPtr=&/LMG_PowerUp_Fire_LoopSlow_FireLoop_Distant_01}
 }
]
properties {
 bDebug=true
 maxDistance=27500
 volumeMultiplier=0.27
 concurrencyOverrides {bLimitToOwner=true, maxCount=3, resolutionRule=StopOldest}
 concurrencySet [&Weapon_Gunfire_Global_Conc, &Weapon_Gunfire_Local_Conc]
}
