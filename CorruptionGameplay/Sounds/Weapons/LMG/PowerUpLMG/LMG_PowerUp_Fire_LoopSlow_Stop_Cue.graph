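# Sound cue for the stop of the power-up LMG's slow firing loop.
# It mixes close, medium and distant layers, each with a single wave, fixed pitch shift, own attenuation and sound class override.
# The close layer uses the third-person fire class; medium and distant share the distant class and reuse LoopFast stop waves pitched down.
# Global and local gunfire concurrency apply, and debug is enabled.
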SoundCue /CorruptionGameplay/Sounds/Weapons/LMG/PowerUpLMG/LMG_PowerUp_Fire_LoopSlow_Stop_Cue
root=Mixer_17
nodes [
 {
  id=Attenuation_0
  children [Modulator_2]
  properties {attenuationSettings=&Fort_Weapon_Projectile_Medium_Mono_Athena_Att}
 }
 {
  id=Attenuation_34
  children [Modulator_12]
  properties {attenuationSettings=&Fort_Weapon_Projectile_Distant_Mono_Athena_Att}
 }
 {
  id=Attenuation_49
  children [WavePlayer_1]
  properties {attenuationSettings=&Fort_Weapon_Projectile_Close_Athena_Att}
 }
 {id=Mixer_17, children [SoundClass_32, SoundClass_0, SoundClass_35], properties {inputVolume [1.05,0.45,0.25]}}
 {id=Modulator_0, children [Attenuation_49], properties {pitchMax=1.2, pitchMin=1.2, volumeMax=0.85, volumeMin=0.8}}
 {id=Modulator_12, children [WavePlayer_5], properties {pitchMax=0.9, pitchMin=0.9, volumeMax=2.05, volumeMin=1.95}}
 {id=Modulator_2, children [WavePlayer_4], properties {pitchMax=0.9, pitchMin=0.9, volumeMax=2.05, volumeMin=1.95}}
 {
  id=SoundClass_0
  children [Attenuation_0]
  properties {soundClassOverride=&Fort_WeaponRanged_Fire_Distant}
 }
 {
  id=SoundClass_32
  children [Modulator_0]
  properties {soundClassOverride=&Fort_WeaponRanged_Fire_3P_PSM}
 }
 {
  id=SoundClass_35
  children [Attenuation_34]
  properties {soundClassOverride=&Fort_WeaponRanged_Fire_Distant}
 }
 {id=WavePlayer_1, properties {soundWaveAssetPtr=&/LMG_PowerUp_Fire_LoopSlow_Stop_03}}
 {id=WavePlayer_4, properties {soundWaveAssetPtr=&/LMG_PowerUp_Fire_LoopFast_Stop_Medium_01}}
 {id=WavePlayer_5, properties {soundWaveAssetPtr=&/LMG_PowerUp_Fire_LoopFast_Stop_Distant_01}}
]
properties {
 bDebug=true
 maxDistance=27500
 volumeMultiplier=0.25
 concurrencySet [&Weapon_Gunfire_Global_Conc, &Weapon_Gunfire_Local_Conc]
}
