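# Sound cue for the end of the power-up LMG reload, classed as weapon foley.
# A mixer layers an LMG equip wave, played immediately, with a slightly quieter cube pickup wave that starts after a short fixed delay.
# The cue plays at low overall volume with shared weapon-reload attenuation, overrides concurrency, and has its debug flag left on.
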SoundCue /CorruptionGameplay/Sounds/Weapons/LMG/PowerUpLMG/LMG_PowerUp_Reload_End_Cue
root=Mixer_0
nodes [
 {id=Delay_0, children [WavePlayer_7], properties {delayMax=0.1, delayMin=0.1}}
 {id=Mixer_0, children [WavePlayer_3,Delay_0], properties {inputVolume [1,0.8]}}
 {id=WavePlayer_3, properties {soundWaveAssetPtr=&/LMG_PowerUp_Equip_04}}
 {id=WavePlayer_7, properties {soundWaveAssetPtr=&/LMG_Powerup_Cube_Pickup_01}}
]
properties {
 attenuationSettings=&Fort_Weapon_Reload_Att
 bDebug=true
 bOverrideConcurrency=true
 maxDistance=2250
 soundClassObject=&Fort_WeaponFoley
 volumeMultiplier=0.15
}
