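# Sound cue for the scope zoom-in of the power-up LMG, routed to the weapon foley sound class.
# A mixer layers a random player-run foley wave at reduced volume with a random scope zoom-in wave at full volume.
# Each layer passes through its own modulator, with a raised pitch ceiling on the zoom layer.
# The cue plays quietly, with overridden short-range attenuation and a concurrency rule that prevents new instances at the limit.
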SoundCue /CorruptionGameplay/Sounds/Weapons/LMG/PowerUpLMG/LMG_PowerUp_Scope_ZoomIn_Cue
root=Mixer_0
nodes [
 {id=Mixer_0, children [Modulator_0,Modulator_1], properties {inputVolume [0.4,1]}}
 {id=Modulator_0, children [Random_4]}
 {id=Modulator_1, children [Random_1], properties {pitchMax=1.3}}
 {id=Random_1, children [WavePlayer_0,WavePlayer_1,WavePlayer_2], properties {weights [1,1,1]}}
 {id=Random_4, children [WavePlayer_19, WavePlayer_1565, WavePlayer_1563, WavePlayer_1562], properties {weights [1,1,1,1]}}
 {id=WavePlayer_0, properties {soundWaveAssetPtr=/Game/Sounds/Fort_Foley_Sounds/Moves/scope_zoom_in_04}}
 {id=WavePlayer_1, properties {soundWaveAssetPtr=/Game/Sounds/Fort_Foley_Sounds/Moves/scope_zoom_in_01}}
 {id=WavePlayer_1562, properties {soundWaveAssetPtr=/Game/Sounds/Fort_Foley_Sounds/Moves/player_run_foley_12}}
 {id=WavePlayer_1563, properties {soundWaveAssetPtr=/Game/Sounds/Fort_Foley_Sounds/Moves/player_run_foley_10}}
 {id=WavePlayer_1565, properties {soundWaveAssetPtr=/Game/Sounds/Fort_Foley_Sounds/Moves/player_run_foley_03}}
 {id=WavePlayer_19, properties {soundWaveAssetPtr=/Game/Sounds/Fort_Foley_Sounds/Moves/player_run_foley_09}}
 {id=WavePlayer_2, properties {soundWaveAssetPtr=/Game/Sounds/Fort_Foley_Sounds/Moves/scope_zoom_in_02}}
]
properties {
 bOverrideAttenuation=true
 bOverrideConcurrency=true
 maxDistance=2700
 soundClassObject=&Fort_WeaponFoley
 volumeMultiplier=0.15
 attenuationOverrides {falloffDistance=1500, nonSpatializedRadiusStart=200, attenuationShapeExtents {x=1200}}
 concurrencyOverrides {resolutionRule=PreventNew}
}
